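; Rod lighting actor that shapes assigned directional, spot, point and rect lights through a sphere, box or cylinder light-function material.
; UpdateROD, run at construction and every Tick, shows the mesh for the rod type, creates a fresh dynamic material instance and assigns it to each light.
; Root transform, shape, hardness, intensity, inner/outer, falloff and invert feed the material only inside the directional-light loop, so other lights rely on one being assigned.
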(event EventBeginPlay)

(event Collision|EventActorBeginOverlap
  (params (OtherActor :type Actor%20Object%20Reference)))

(event EventTick
  (params (DeltaSeconds :type Float%20%28single-precision%29))
  (CallFunction|UpdateROD))

(fn update_ROD
  (bind (_returnvalue_roll _returnvalue_pitch _returnvalue_yaw) (Math|Rotator|BreakRotator (Transformation|GetWorldRotation (Variables|Default|GetDefaultSceneRoot))))
  (bind _mid (Variables|Default|GetMID))
  (bind _directionallight (Variables|Default|GetDirectionalLight))
  (bind _mid_2 (Variables|Default|GetMID))
  (bind _spotlight (Variables|Default|GetSpotLight))
  (bind _pointlight (Variables|Default|GetPointLight))
  (bind _rectlight (Variables|Default|GetRectLight))
  (switch Utilities|FlowControl|Switch|SwitchonRodType2 (Variables|Default|GetR_Type)
    (:NewEnumerator0
      (Rendering|SetVisibility (Variables|Default|GetSphere) true false))
    (:NewEnumerator1
      (Rendering|SetVisibility (Variables|Default|GetBox) true false))
    (:NewEnumerator2
      (Rendering|SetVisibility (Variables|Default|GetCylinder) true false)))
  (bind _returnvalue (Rendering|Material|CreateDynamicMaterialInstance "/Game/Blueprints/Lighting/ROD/Light_Function_shapes_ROD.Light_Function_shapes_ROD" "None" "None" :!class /Script/Engine.KismetMaterialLibrary))
  (Variables|Default|SetMID _returnvalue)
  (if (!= (Utilities|Array|Length _directionallight) 0)
    (for _array_element _directionallight
      (Utilities|IsValid _array_element
        (:Is%20Valid
          (Rendering|Components|Light|SetLightFunctionMaterial (Variables|/Script/Engine.Light|GetLightComponent _array_element) _mid_2)
          (Rendering|Material|SetVectorParameterValue _mid_2 "Translate" (Math|Conversions|ToLinearColor(Vector) (Transformation|GetWorldLocation (Variables|Default|GetDefaultSceneRoot))) :!class /Script/Engine.MaterialInstanceDynamic)
          (Rendering|Material|SetVectorParameterValue _mid_2 "Rotate" (Utilities|Struct|MakeLinearColor _returnvalue_roll _returnvalue_pitch _returnvalue_yaw 1) :!class /Script/Engine.MaterialInstanceDynamic)
          (Rendering|Material|SetVectorParameterValue _mid_2 "Scale" (Math|Conversions|ToLinearColor(Vector) (Transformation|GetWorldScale (Variables|Default|GetDefaultSceneRoot))) :!class /Script/Engine.MaterialInstanceDynamic)
          (switch Utilities|FlowControl|Switch|SwitchonRodType2 (Variables|Default|GetR_Type)
            (:NewEnumerator0
              (Rendering|Material|SetScalarParameterValue _mid "Shape" 0.0 :!class /Script/Engine.MaterialInstanceDynamic))
            (:NewEnumerator1
              (Rendering|Material|SetScalarParameterValue _mid "Shape" 1.0 :!class /Script/Engine.MaterialInstanceDynamic))
            (:NewEnumerator2
              (Rendering|Material|SetScalarParameterValue _mid "Shape" 2.0 :!class /Script/Engine.MaterialInstanceDynamic)))
          (Rendering|Material|SetScalarParameterValue _mid_2 "Hardness" (Variables|Default|GetHardness) :!class /Script/Engine.MaterialInstanceDynamic)
          (Rendering|Material|SetScalarParameterValue _mid_2 "Intensity" (Variables|Default|GetIntensity) :!class /Script/Engine.MaterialInstanceDynamic)
          (Rendering|Material|SetScalarParameterValue _mid_2 "Inner" (Variables|Default|GetInner) :!class /Script/Engine.MaterialInstanceDynamic)
          (Rendering|Material|SetScalarParameterValue _mid_2 "Outter" (Variables|Default|GetOutter) :!class /Script/Engine.MaterialInstanceDynamic)
          (Rendering|Material|SetScalarParameterValue _mid_2 "Falloff" (Variables|Default|GetFalloff) :!class /Script/Engine.MaterialInstanceDynamic)
          (Rendering|Material|SetScalarParameterValue _mid_2 "Invert" (Math|Conversions|ToFloat(Boolean) (Variables|Default|GetInvert)) :!class /Script/Engine.MaterialInstanceDynamic)))))
  (if (!= (Utilities|Array|Length _spotlight) 0)
    (for _array_element_2 _spotlight
      (Utilities|IsValid _array_element_2
        (:Is%20Valid
          (Rendering|Components|Light|SetLightFunctionMaterial (Variables|/Script/Engine.Light|GetLightComponent _array_element_2) _mid_2)))))
  (if (!= (Utilities|Array|Length _pointlight) 0)
    (for _array_element_3 _pointlight
      (Utilities|IsValid _array_element_3
        (:Is%20Valid
          (Rendering|Components|Light|SetLightFunctionMaterial (Variables|/Script/Engine.Light|GetLightComponent _array_element_3) _mid_2)))))
  (if (!= (Utilities|Array|Length _rectlight) 0)
    (for _array_element_4 _rectlight
      (Utilities|IsValid _array_element_4
        (:Is%20Valid
          (Rendering|Components|Light|SetLightFunctionMaterial (Variables|/Script/Engine.Light|GetLightComponent _array_element_4) _mid_2))))))

(fn ConstructionScript
  (CallFunction|UpdateROD))
