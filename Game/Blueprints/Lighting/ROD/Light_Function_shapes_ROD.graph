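# Light function material that masks a light with a parameter-driven sphere, box or cylinder shape placed in world space.
# World position is offset by Translate, rotated about Z, Y and X by Rotate, then divided by scaled Scale.
# SmoothStep between Inner and Outer builds each shape's falloff, and Shape blends them through Lerp_Multiple_Float.
# A SphereMask applies Hardness, Invert flips the mask, and the result blends between Intensity and full light.
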Material /Game/Blueprints/Lighting/ROD/Light_Function_shapes_ROD
graph {
 properties {materialDomain=MD_LightFunction, pixelDepthOffsetMode=PDOM_Legacy, usedShadingModels=Default Lit}
 nodes [
  n1 = VectorParameter() {parameterName=Rotate}
  ComponentMask(None=@n1.RGB) {b=true}
  ComponentMask(None=@n1.RGB) {g=true}
  ComponentMask(None=@n1.RGB) {r=true}
  Constant() {r=1}
  n2 = Constant3Vector() {constant=<rgba 1 0 0 0>}
  n3 = Multiply(A=@n1.R)
  n4 = Constant3Vector()
  n5 = Subtract(A=WorldPosition().XYZ, B=ComponentMask(None=VectorParameter() {parameterName=Translate}.RGB) {b=true, g=true, r=true})
  n6 = Add(
   A=RotateAboutAxis(
    NormalizedRotationAxis=Constant3Vector() {constant=<rgba 0 0 1 0>}
    RotationAngle=Multiply(A=@n1.B) {constB=-1}
    PivotPoint=@n4
    Position=@n5
   ) {period=360}
   B=@n5
  )
  n7 = Add(
   A=RotateAboutAxis(NormalizedRotationAxis=Constant3Vector() {constant=<rgba 0 1 0 0>}, RotationAngle=Multiply(A=@n1.G), PivotPoint=@n4, Position=@n6) {period=360}
   B=@n6
  )
  n8 = VectorParameter() {parameterName=Scale}
  n9 = Divide(
   A=Add(A=RotateAboutAxis(NormalizedRotationAxis=@n2, RotationAngle=@n3, PivotPoint=@n4, Position=@n7) {period=360}, B=@n7)
   B=Multiply(A=ComponentMask(None=@n8.RGB) {b=true, g=true, r=true}) {constB=75}
  )
  n10 = Distance(A=@n9, B=Constant3Vector())
  n11 = ScalarParameter() {defaultValue=0.5, parameterName=Inner}
  n12 = ScalarParameter() {defaultValue=1, parameterName=Outer}
  n13 = ComponentMask(None=@n9) {r=true}
  n14 = Multiply(A=@n11) {constB=-1}
  n15 = Multiply(A=@n12) {constB=-1}
  n16 = Max(
   A=MaterialFunctionCall(Alpha=@n13, Min=@n11, Max=@n12) {materialFunction=&SmoothStep}.Result
   B=MaterialFunctionCall(Alpha=@n13, Min=@n14, Max=@n15) {materialFunction=&SmoothStep}.Result
  )
  n17 = ComponentMask(None=@n9) {g=true}
  n18 = MaterialFunctionCall(Alpha=@n17, Min=@n11, Max=@n12) {materialFunction=&SmoothStep}
  n19 = ComponentMask(None=@n9) {b=true}
  n20 = Max(
   A=Max(
    A=Max(A=@n16, B=@n18.Result)
    B=MaterialFunctionCall(Alpha=@n17, Min=@n14, Max=@n15) {materialFunction=&SmoothStep}.Result
   )
   B=MaterialFunctionCall(Alpha=@n19, Min=@n11, Max=@n12) {materialFunction=&SmoothStep}.Result
  )
  n21 = MaterialFunctionCall(Alpha=@n19, Min=@n14, Max=@n15) {materialFunction=&SmoothStep}
  n22 = Multiply(A=@n9, B=Constant3Vector() {constant=<rgba 1 1 0 0>})
  n23 = Constant3Vector()
  n24 = ComponentMask(None=@n9) {b=true}
  n25 = Max(
   A=MaterialFunctionCall(Alpha=Distance(A=@n22, B=@n23), Min=@n11, Max=@n12) {materialFunction=&SmoothStep}.Result
   B=MaterialFunctionCall(Alpha=@n24, Min=@n11, Max=@n12) {materialFunction=&SmoothStep}.Result
  )
  n26 = MaterialFunctionCall(Alpha=@n24, Min=Multiply(A=@n11) {constB=-1}, Max=Multiply(A=@n12) {constB=-1}) {materialFunction=&SmoothStep}
  n27 = ScalarParameter() {parameterName=Shape}
  n28 = SphereMask(
   A=MaterialFunctionCall!(
    1=MaterialFunctionCall(Alpha=@n10, Min=@n11, Max=@n12) {materialFunction=&SmoothStep}.Result
    2=Max(A=@n20, B=@n21.Result)
    3=Max(A=@n25, B=@n26.Result)
    4
    A=Divide(A=@n27)
   ) {materialFunction=/Engine/Functions/Engine_MaterialFunctions02/Lerp_Multiple_Float}.Lerp 3 Inputs
   B=Constant3Vector()
   Radius=Constant() {r=1}
   Hardness=ScalarParameter() {defaultValue=1, parameterName=Hardness}
  )
  LinearInterpolate(
   A=ScalarParameter() {defaultValue=1, parameterName=Intensity}
   B=Constant() {r=1}
   Alpha=LinearInterpolate(A=OneMinus(None=@n28), B=@n28, Alpha=ScalarParameter() {defaultValue=1, parameterName=Invert})
  )
 ]
}
