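# M_Rod is a translucent, unlit material stored with the ROD lighting Blueprint assets.
# Its graph holds only a Constant3Vector and a scalar Constant, both at default values, with no connections to material outputs shown.
# It has no parameters, textures or other logic.
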Material /Game/Blueprints/Lighting/ROD/M_Rod
graph {
 properties {blendMode=BLEND_Translucent, pixelDepthOffsetMode=PDOM_Legacy, shadingModel=MSM_Unlit, usedShadingModels=Unlit}
 nodes [
  Constant3Vector()
  Constant()
 ]
}
